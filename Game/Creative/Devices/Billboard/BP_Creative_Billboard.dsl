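; Creative Billboard device that shows configurable text on a widget component, with an optional border mesh.
; Option changes re-run initialization, pushing style, text and view distance to the runtime-created widget on cosmetic-capable instances.
; The server drives the enabled component from minigame phase, tracking the last state to preserve manual changes; this controls text and border visibility.
; Events can show, hide or override the text, which refreshes through OnRep and requested widget redraws.
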(event Custom|Update%20Static%20Mesh%20Component%20Visibility%20Options
  (bind _loaded_object (Creative|Loading|AsyncLoadEditorSafe (select (Variables|UserOptions|GetShowBorder) (Variables|Default|GetDefaultMaterial) (Variables|Default|GetClearMaterial)))
    (:Loaded
      (bind _asmaterial_interface (Utilities|Casting|CastToMaterialInterface _loaded_object))
      (bind _output_get (Variables|Default|SetCachedMaterial _asmaterial_interface))
      (Rendering|Material|SetMaterial (Variables|Debug|GetStaticMeshComponent) 0 _output_get)
      (bind _widgetanchor (Variables|Default|GetWidgetAnchor)
        (:then
          (Transformation|SetRelativeRotation _widgetanchor (select (Variables|UserOptions|GetShowBorder) (Variables|Default|GetBackgroundRelativeRotationWithBorder) (Variables|Default|GetBackgroundRelativeRotationWithoutBorder)) false false))
        (:else)))))

(event EventOnWorldReady
  (Collision|SetCollisionEnabled (Variables|Debug|GetStaticMeshComponent) (select (Variables|UserOptions|GetShowBorder) "QueryAndPhysics" "NoCollision"))
  (CallFunction|UpdateBorderVisibility (Class|CreativeEnabledComponent|IsEnabled (Variables|Default|GetCreative_Enabled_Component))))

(event Custom|Initialize%20Device
  (Utilities|FlowControl|Sequence
    (:then_0
      (Class|CreativeEnabledComponent|Initialize (Variables|Default|GetCreative_Enabled_Component) (Variables|Default|GetFortMinigameLogic) (Variables|UserOptions|GetEnabledDuringPhase))
      (Collision|SetCollisionEnabled (Variables|Debug|GetStaticMeshComponent) (select (Variables|UserOptions|GetShowBorder) "QueryAndPhysics" "NoCollision"))
      (CallFunction|UpdateBorderVisibility (Class|CreativeEnabledComponent|IsEnabled (Variables|Default|GetCreative_Enabled_Component))))
    (:then_1
      (CallFunction|UpdateProperties)
      (bind _currentminigame (Variables|/Script/FortniteGame.FortMinigameLogicComponent|GetCurrentMinigame (Variables|Default|GetFortMinigameLogic))
        (:then
          (CallFunction|InitializeMinigameOptions _currentminigame))
        (:else
          (if (Editor|InEditor)
            (CallFunction|TextVisibilityonGameState "PreGame")
            (else
              (Minigame|BindEventtoOnAddedtoMinigame (Variables|Default|GetFortMinigameLogic) (delegate Custom|Initialize%20Minigame%20Options)) ; data loaded before minigame. wait to initialize minigame options until minigame is valid.
            )))))))

(event EditorPreview|EventSetupForEditorPreview
  (|CallParentFunction)
  (CallFunction|CreateWidget) ; Only create the widget once, after BeginPlay has happened or it will fail to be created.
  (CallFunction|InitializeDevice))

(event EventBeginPlay
  (CallFunction|CreateWidget) ; Only create the widget once, after BeginPlay has happened or it will fail to be created.
  (CallFunction|InitializeDevice))

(event OnPropertiesLoaded(ToyOptionsComponent)
  (CallFunction|InitializeDevice))

(event OnAnyPropertyChanged(ToyOptionsComponent)
  (CallFunction|InitializeDevice))

(event MinigameStateChangedDelegate(FortMinigameLogic)
  (params (Minigame :type Fort%20Minigame%20Object%20Reference) (NewMinigameState :type EFortMinigameState%20Enum :default "PreGame"))
  (CallFunction|TextVisibilityonGameState NewMinigameState))

(event Custom|Initialize%20Minigame%20Options
  (params (Minigame :type Fort%20Minigame%20Object%20Reference))
  (CallFunction|PlayModeChanged Minigame (Minigame|IsinPlayMode Minigame :!class /Script/FortniteGame.FortMinigame))
  (Networking|SwitchHasAuthority
    (:Authority
      (CallFunction|TextVisibilityonGameState (Variables|/Script/FortniteGame.FortMinigame|GetCurrentState Minigame)))))

(event Custom|PostUpdateProperties
  (Networking|CanExecuteCosmeticEvents
    (:True
      (bind _widgetcomponent (Variables|Default|GetWidgetComponent)
        (:then
          (UserInterface|RequestRenderUpdate _widgetcomponent))
        (:else)))))

(event OnEnabledStateChanged(Creative_Enabled_Component)
  (params (Enabled :type Boolean))
  (CallFunction|UpdateLastGameStateEnableChanged)
  (UpdateProperties|SetDisplayText)
  (CallFunction|PostUpdateProperties)
  (CallFunction|UpdateBorderVisibility Enabled))

(event OnAddedtoMinigame(FortMinigameLogic)
  (params (Minigame :type Fort%20Minigame%20Object%20Reference))
  (Utilities|IsValid Minigame
    (:Is%20Valid
      (Minigame|BindEventtoOnPlayModeChangedDelegate Minigame (delegate Custom|Play%20Mode%20Changed)))))

(event Custom|Play%20Mode%20Changed
  (params (Minigame :type Fort%20Minigame%20Object%20Reference) (bIsInPlayMode :type Boolean))
  (bind (_backgroundcolor_r _backgroundcolor_g _backgroundcolor_b _backgroundcolor_a) (Math|Color|BreakColor (Variables|GetBackgroundColor_R)))
  (Collision|SetCollisionResponsetoChannel (Variables|Debug|GetStaticMeshComponent) "ECC_GameTraceChannel1" (select (not bIsInPlayMode) "ECR_Block" (select (Math|Float|NearlyEqual(Float) _backgroundcolor_a) "ECR_Ignore" "ECR_Block")))
  (Collision|SetCollisionEnabled (Variables|Default|GetToolPickingBox) (select bIsInPlayMode "NoCollision" "QueryOnly")) ; Don't want the Tool Picker box to have any \neffect during play (it can interrupt interacting\nwith other devices that are intersecting the \nbillboard.
  (if (not bIsInPlayMode)
    (Variables|Default|SetText%20From%20Event "")))

(event Custom|ReceiveShowTextEvent
  (params (Instigator :type Controller%20Object%20Reference))
  (Class|CreativeEnabledComponent|SetEnabled (Variables|Default|GetCreative_Enabled_Component) true))

(event Custom|ReceiveHideTextEvent
  (params (Instigator :type Controller%20Object%20Reference))
  (Class|CreativeEnabledComponent|SetEnabled (Variables|Default|GetCreative_Enabled_Component) false))

(event Custom|Update%20Display%20Event
  (params (Instigator :type Controller%20Object%20Reference) (Type :type EPayload_DataType%20Enum :default "NewEnumerator0") (Text :type Text) (Number :type Float%20%28double-precision%29))
  ;[1 0.759 0.6475] Update String Payload from Remote Source
  (if (Utilities|Enum|NotEqual(Enum) Type "NewEnumerator0")
    (Variables|Default|SetText%20From%20Event (Utilities|String|SelectString (Utilities|String|ToString(Text) Text) (Utilities|String|ToString(Integer64) (Math|Float|TruncatetoInteger64 Number)) (Utilities|Enum|Equal(Enum) Type "NewEnumerator1")))))

(fn CreateWidget
  (Networking|CanExecuteCosmeticEvents
    (:True
      (bind _returnvalue (AddComponent|NODE_AddWidgetComponent-0 :bManualAttachment true)) ; Spawn a Creative_Background_Widget
      (bind _output_get (Variables|Default|SetWidgetComponent _returnvalue))
      (Animation|SetTickwhenOffscreen _output_get true)
      (bind _ascreative_background_widget (Utilities|Casting|CastToCreative_Background_Widget (UserInterface|GetWidget _output_get)))
      (Variables|Default|SetWidget _ascreative_background_widget)
      (Transformation|AttachComponentToComponent (Variables|Default|GetWidgetComponent) (Variables|Default|GetWidgetAnchor) "None" "SnapToTarget" "SnapToTarget" "SnapToTarget" true) ; Attach our widget component to the anchor location\nand snap to it.
    )))

(fn GetCollisionStaticMesh
  (return "0, 0, 0" (Variables|/Script/Engine.StaticMeshComponent|GetStaticMesh (Variables|Debug|GetStaticMeshComponent))))

(fn GetCreativeActorBounds
  (bind (_origin _boxextent) (Collision|GetComponentBounds (Variables|Default|GetCreativePreviewOnlyMesh)))
  ;[1 1 1] Due to the root mesh having the pivot at the bottom use the CreativePreviewOnlyMesh instead.
  (return true _boxextent))

(fn GetCreativeActorOrigin
  ;[1 1 1] Override the origin with the widget location due to the root mesh having the pivot at the bottom.
  (return true (Variables|/Script/Engine.SceneComponent|GetRelativeLocation (Variables|Default|GetWidgetAnchor))))

(fn GetCreativeActorRelativeTransform
  (return false 0))

(fn GetGridAxisGuideAttachmentPoint
  (return false false "0, 0, 0"))

(fn GetMeshComponents
  (return (Utilities|Array|MakeArray (Variables|Debug|GetStaticMeshComponent))))

(fn GetOverrideMeshMaterial
  (params (MeshComp :type Static%20Mesh%20Component%20Object%20Reference) (MatIdx :type Integer))
  (if (Utilities|Equal(Object) MeshComp (Variables|Debug|GetStaticMeshComponent))
    (return (Variables|Default|GetCachedMaterial))
    (else
      (return 0))))

(fn IgnoreWorldGridScaleOverride
  (return false))

(fn IsTargetableByLinearMover
  (return true))

(fn OnRep_TextFromEvent
  (UpdateProperties|SetDisplayText)
  (CallFunction|PostUpdateProperties))

(fn SetBackgroundColor
  (bind _widget (Variables|Default|GetWidget)
    (:then
      (Class|CreativeBackgroundWidget|SetBackgroundColor _widget (Math|Color|SelectColor "(R=0.000000,G=0.000000,B=0.000000,A=0.000000)" (Variables|UserOptions|Basic|GetBackgroundColor) (Variables|UserOptions|GetShowBorder))))
    (:else)))

(fn SetDisplayText
  (bind _widget (Variables|Default|GetWidget)
    (:then
      (Class|CreativeBackgroundWidget|SetText _widget (select (Utilities|String|IsEmpty (Variables|Default|GetText%20From%20Event)) (Variables|UserOptions|Basic|GetText) (Variables|Default|GetText%20From%20Event)))
      (if (Class|CreativeEnabledComponent|IsEnabled (Variables|Default|GetCreative_Enabled_Component))
        (Widget|SetVisibility _widget "Visible")
        (else
          (Widget|SetVisibility _widget "Hidden"))))
    (:else)))

(fn SetShadow
  (bind _widget (Variables|Default|GetWidget)
    (:then
      (Class|CreativeBackgroundWidget|SetShadow _widget (Variables|UserOptions|Basic|GetShadow) (/ (Math|Conversions|ToFloat(Integer) (Variables|UserOptions|Basic|GetTextSize)) (Variables|Default|GetShadow%20Divisor))))
    (:else)))

(fn SetTextAlignment
  (bind _widget (Variables|Default|GetWidget)
    (:then
      (Class|CreativeBackgroundWidget|SetTextAlignment _widget (Variables|UserOptions|Basic|GetTextJustification)))
    (:else)))

(fn SetTextColor
  (bind _widget (Variables|Default|GetWidget)
    (:then
      (Class|CreativeBackgroundWidget|SetTextColor _widget (Variables|UserOptions|Basic|GetTextColor)))
    (:else)))

(fn SetTextFont
  (bind _widget (Variables|Default|GetWidget)
    (:then
      (Class|CreativeBackgroundWidget|SetTextFont _widget (Math|Conversions|ToInteger(Byte) (Variables|UserOptions|Basic|GetTextFont)) (Math|Float|Truncate (* (Math|Conversions|ToInteger(Byte) (Variables|UserOptions|Basic|GetOutline)) (/ (Math|Conversions|ToFloat(Integer) (Variables|UserOptions|Basic|GetTextSize)) (Variables|Default|GetMinTextSize))))))
    (:else)))

(fn SetTextSize
  (bind _widget (Variables|Default|GetWidget)
    (:then
      (Class|CreativeBackgroundWidget|SetTextScale _widget (Variables|UserOptions|Basic|GetTextSize)))
    (:else)))

(fn SetViewDistance
  (bind _viewdistance (Variables|UserOptions|GetViewDistance))
  (bind _returnvalue (* _viewdistance 512.0))
  (LOD|SetMaxDrawDistance (Variables|Debug|GetStaticMeshComponent) _returnvalue)
  (LOD|SetMaxDrawDistance (Variables|Default|GetWidgetComponent) (- _returnvalue 1e+02)))

(fn ShouldInvisiblePropsBeAllowed
  (return true))

(fn TextVisibilityOnGameState
  (params (New%20State :type EFortMinigameState%20Enum :default "PreGame"))
  (bind _fortminigamelogic (Variables|Default|GetFortMinigameLogic))
  (bind _currentminigame (Variables|/Script/FortniteGame.FortMinigameLogicComponent|GetCurrentMinigame _fortminigamelogic))
  (Networking|SwitchHasAuthority
    (:Authority
      (if (Math|Byte|NotEqual(Byte) New%20State (Variables|Default|GetLast%20State%20Updated))
        (Utilities|FlowControl|Sequence
          (:then_0
            (switch Utilities|FlowControl|Switch|SwitchonEEnabledOnPhase (Variables|UserOptions|GetEnabledDuringPhase)
              (:NewEnumerator1
                (LocalVariables|SetbIsTextVisible false))
              (:NewEnumerator0
                (LocalVariables|SetbIsTextVisible true))
              (:NewEnumerator2
                ;[1 1 1] Warmup Only
                (if (Utilities|Enum|NotEqual(Enum) (Variables|GetNew%20State) "PreGame")
                  (LocalVariables|SetbIsTextVisible false)
                  (else
                    (LocalVariables|SetbIsTextVisible true))))
              (:NewEnumerator4
                ;[1 1 1] Gameplay Only
                (if (Utilities|Enum|NotEqual(Enum) (Variables|GetNew%20State) "InProgress")
                  (LocalVariables|SetbIsTextVisible false)
                  (else
                    (LocalVariables|SetbIsTextVisible true))))
              (:NewEnumerator3
                ;[1 1 1] FORT-524469	| Show Text on Create Only
                (Utilities|IsValid _fortminigamelogic
                  (:Is%20Valid
                    (Utilities|IsValid _currentminigame
                      (:Is%20Valid
                        (LocalVariables|SetbIsTextVisible (not (Minigame|IsinPlayMode _currentminigame :!class /Script/FortniteGame.FortMinigame))) ; Only shows the text if not in Play mode
                      )))))))
          (:then_1
            (Class|CreativeEnabledComponent|SetEnabled (Variables|Default|GetCreative_Enabled_Component) (LocalVariables|GetbIsTextVisible))))))))

(fn UpdateBorderVisibility
  (params (bNewVisibility :type Boolean))
  ;[1 1 1] Set border mesh visibility along with EnabledDuringPhase Settings
  (Rendering|SetVisibility (Variables|Debug|GetStaticMeshComponent) bNewVisibility false)
  ;[1 1 1] Enable Border collision only when "the device is enabled" & "show border is true"
  (Collision|SetCollisionEnabled (Variables|Debug|GetStaticMeshComponent) (select (and (Variables|UserOptions|GetShowBorder) bNewVisibility) "QueryAndPhysics" "NoCollision")))

(fn UpdateLastGameStateEnableChanged
  (Networking|SwitchHasAuthority
    (:Authority
      (bind _fortminigamelogic (Variables|Default|GetFortMinigameLogic)
        (:then
          (bind _currentminigame (Variables|/Script/FortniteGame.FortMinigameLogicComponent|GetCurrentMinigame _fortminigamelogic)
            (:then
              (Variables|Default|SetLast%20State%20Updated (Variables|/Script/FortniteGame.FortMinigame|GetCurrentState _currentminigame)))
            (:else)))
        (:else)))))

(fn UpdateProperties
  (CallFunction|UpdateStaticMeshComponentVisibilityOptions)
  (Networking|CanExecuteCosmeticEvents
    (:True
      (Utilities|IsValid (Variables|Default|GetWidget)
        (:Is%20Valid
          (UpdateProperties|SetTextColor)
          (UpdateProperties|SetViewDistance)
          (UpdateProperties|SetBackgroundColor)
          (UpdateProperties|SetTextAlignment)
          (UpdateProperties|SetDisplayText)
          (UpdateProperties|SetTextSize)
          (UpdateProperties|SetTextFont)
          (UpdateProperties|SetShadow)
          (CallFunction|PostUpdateProperties)
          (Development|PrintString "BillBoard: Client" (Variables|Default|GetShow%20Debug%20Text) (Variables|Default|GetShow%20Debug%20Text)))) ; If our widget isn't created yet (aka before begin play is called) No need to set anything because these all set things on the widget. They'll get set once the widget is created.
    )
    (:False
      (Variables|Default|SetLast%20State%20Updated "PostGameReset") ; Last State Updated is used to avoid the Visibility on Game State\nfunction switching back if a manual change has occurred first. \nWhen we initialize properties, set this to a phase that is outside the normal \nparameters for this device.
      (Development|PrintString "Billboard:Server" (Variables|Default|GetShow%20Debug%20Text) (Variables|Default|GetShow%20Debug%20Text))))
  (Development|PrintString (Utilities|String|ToString(Text) (Utilities|Text|FormatText "NSLOCTEXT(\"[3C5963E546683DC11330588A9D9B5A15]\", \"AC120DD44C93F8FFC24DE9986304631C\", \"Text: {Text}, Border {Border}, BackColor {BackColor}, TextColor {TextColor}, Align {Align}, TextSize {TextSize}, ViewDist {ViewDist}\")" (Utilities|Text|ToText(String) (Variables|UserOptions|Basic|GetText)) (Utilities|Text|ToText(Boolean) (Variables|UserOptions|GetShowBorder)) (Utilities|Text|ToText(LinearColor) (Variables|UserOptions|Basic|GetBackgroundColor)) (Utilities|Text|ToText(LinearColor) (Variables|UserOptions|Basic|GetTextColor)) (Utilities|Text|ToText(String) (Utilities|String|ToString(Byte) (Variables|UserOptions|Basic|GetTextJustification))) (Variables|UserOptions|Basic|GetTextSize) (Variables|UserOptions|GetViewDistance))) (Variables|Default|GetShow%20Debug%20Text) (Variables|Default|GetShow%20Debug%20Text) "(R=0.000000,G=0.660000,B=1.000000,A=1.000000)" 1e+01))

(fn ConstructionScript)
