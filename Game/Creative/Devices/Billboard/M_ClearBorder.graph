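# M_ClearBorder is an additive material for the Creative Billboard device, apparently a see-through border as its name suggests.
# Its only wired attribute is Opacity, taken from the alpha of the BlueprintColor vector parameter in the CreativeMoveTool group.
# The material attributes pass through MF_CreativeMoveTool, the shared move-tool function.
# It allows translucent custom depth writes and is flagged for Nanite use.
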Material /Game/Creative/Devices/Billboard/M_ClearBorder
graph {
 properties {
  allowTranslucentCustomDepthWrites=true
  bUseMaterialAttributes=true
  bUsedWithNanite=true
  blendMode=BLEND_Additive
  pixelDepthOffsetMode=PDOM_Legacy
  usedShadingModels=Default Lit
 }
 nodes [MaterialFunctionCall(MA In=MakeMaterialAttributes!(
    BaseColor
    Metallic
    Specular
    Roughness
    Anisotropy
    EmissiveColor
    Opacity=VectorParameter() {group=CreativeMoveTool, parameterName=BlueprintColor}.A
    OpacityMask
    Normal
    Tangent
    WorldPositionOffset
    SubsurfaceColor
    ClearCoat
    ClearCoatRoughness
    AmbientOcclusion
    Refraction
    CustomizedUVs_0
    CustomizedUVs_1
    CustomizedUVs_2
    CustomizedUVs_3
    CustomizedUVs_4
    CustomizedUVs_5
    CustomizedUVs_6
    CustomizedUVs_7
    PixelDepthOffset
    ShadingModel
    Displacement
   )) {materialFunction=&MF_CreativeMoveTool}]
}
