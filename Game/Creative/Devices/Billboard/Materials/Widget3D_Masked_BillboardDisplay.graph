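# Masked, unlit material for the Creative Billboard device's display, drawing a 3D widget's SlateUI texture.
# It samples SlateUI with normal and U-flipped UVs, picking by TwoSidedSign so back faces are un-mirrored.
# TintColorAndOpacity multiplies color and alpha; despite its name, OpacityFromTexture blends color from flat tint to tinted texture, leaving alpha untouched.
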Material /Game/Creative/Devices/Billboard/Materials/Widget3D_Masked_BillboardDisplay
graph {
 properties {bCastRayTracedShadows=false, blendMode=BLEND_Masked, pixelDepthOffsetMode=PDOM_Legacy, shadingModel=MSM_Unlit, usedShadingModels=Unlit}
 nodes [
  n1 = MaterialFunctionCall(Float2=TextureCoordinate()) {materialFunction=&BreakOutFloat2Components}
  n2 = TextureSampleParameter2D(UVs=MaterialFunctionCall(X=Subtract(B=@n1.R), Y=@n1.G) {materialFunction=&MakeFloat2}.Result) {parameterName=SlateUI, samplerSource=SSM_Clamp_WorldGroupSettings, texture=/Engine/EngineResources/WhiteSquareTexture}
  n3 = TextureSampleParameter2D() {parameterName=SlateUI, samplerSource=SSM_Clamp_WorldGroupSettings, texture=&WhiteSquareTexture}
  n4 = Saturate(None=TwoSidedSign())
  n5 = VectorParameter() {parameterName=TintColorAndOpacity}
  LinearInterpolate(
   A=@n5.RGB
   B=Multiply(A=LinearInterpolate(A=@n2.RGB, B=@n3.RGB, Alpha=@n4) {constAlpha=1}, B=@n5.RGB)
   Alpha=ScalarParameter() {defaultValue=1, parameterName=OpacityFromTexture, sliderMax=1}
  ) {constAlpha=1}
  Multiply(A=LinearInterpolate(A=@n2.A, B=@n3.A, Alpha=@n4) {constAlpha=1}, B=@n5.A)
 ]
}
